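# Material for the emissive engine glow on the Speedy Midnight melee pickaxe, using the Default Lit shading model.
# A Glow_Color vector parameter, defaulting to an overbright purple, feeds the Emissive input of the shared MF_FX_EmissiveAdjustments function.
# The graph holds no textures or other logic, and usage flags cover skeletal meshes, mesh deformers, Nanite and Niagara mesh particles.
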Material /Game/Weapons/FORT_Melee/Pickaxe_Speedy_Midnight/Materials/M_Speedy_Engine_Emissive
graph {
 properties {
  bUsedWithMeshDeformer=true
  bUsedWithNanite=true
  bUsedWithNiagaraMeshParticles=true
  bUsedWithSkeletalMesh=true
  pixelDepthOffsetMode=PDOM_Legacy
  usedShadingModels=Default Lit
 }
 nodes [MaterialFunctionCall(Emissive=VectorParameter() {parameterName=Glow_Color, defaultValue=<rgba 4 0.2 8 1>}.RGB) {materialFunction=&MF_FX_EmissiveAdjustments}]
}
